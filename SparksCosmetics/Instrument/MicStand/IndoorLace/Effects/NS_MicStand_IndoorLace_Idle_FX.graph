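# Idle Niagara effect for the IndoorLace mic stand cosmetic: warm glows, flickering sparkles and a spiraling ribbon.
# Mic glow spawns depending on User.IsInstrument, sparkles swirl through curl noise and vortex forces, and a burst-spawned ribbon is repositioned each frame into an animated vertical helix.
# Smoothed system reactivity values drive size, brightness and sparkle rate, while time of day and front-end preview adjust glow opacity.
# Stars_and_Motes001 is disabled.
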NiagaraSystem /SparksCosmetics/Instrument/MicStand/IndoorLace/Effects/NS_MicStand_IndoorLace_Idle_FX
dependencies {
 usedDataInterfaces [
  &NiagaraDataInterfaceCurve
  &NiagaraDataInterfaceCamera
  &NiagaraDataInterfaceColorCurve
  &NiagaraDataInterfaceArrayFloat
 ]
 usedDynamicInputs [
  &MakeFloatFromVector
  &MakeLinearColor
  &MakeCustomLinearColorFromBool
 ]
 usedModules [
  &SystemState
  &self:SystemSpawnScript.SystemScriptSource.SystemScriptGraph.NiagaraNodeAssignment_1.SetVariables_#1
  &self:SystemSpawnScript.SystemScriptSource.SystemScriptGraph.NiagaraNodeAssignment_2.SetVariables_#2
  &self:SystemSpawnScript.SystemScriptSource.SystemScriptGraph.NiagaraNodeAssignment_7.SetVariables_#3
  &self:SystemSpawnScript.SystemScriptSource.SystemScriptGraph.NiagaraNodeAssignment_4.SetVariables_#4
  &self:SystemSpawnScript.SystemScriptSource.SystemScriptGraph.NiagaraNodeAssignment_0.SetVariables_#5
  &Timeline
  &self:SystemSpawnScript.SystemScriptSource.SystemScriptGraph.NiagaraNodeAssignment_6.SetVariables_#6
  &self:SystemSpawnScript.SystemScriptSource.SystemScriptGraph.NiagaraNodeAssignment_3.SetVariables_#7
  &InterpolateOverTime
  &self:SystemSpawnScript.SystemScriptSource.SystemScriptGraph.NiagaraNodeAssignment_5.SetVariables_#8
  &self:SystemSpawnScript.SystemScriptSource.SystemScriptGraph.NiagaraNodeAssignment_8.SetVariables_#9
 ]
}
emitters {
 BaseGlows_MicGlow {
  emitterUpdateScript [
   &EmitterState()
   &SpawnBurst_Instantaneous(
    Spawn Count=2
    Spawn Group=1
    Spawn Probability=&MakeCustomFloatFromBool(Bool=&SetBoolByFloatComparison(A=$User.IsInstrument, B=0), True Float=0, False Float=1)
    Use Spawn Probability=true
    UseLoopCountLimit=true
   ) {module=SpawnBurst_Instantaneous001}
   &SpawnBurst_Instantaneous(Spawn Count=5, Use Spawn Probability=false, UseLoopCountLimit=true) {enabled=false}
  ]
  particleSpawnScript [
   &PartitionParticles()
   /SetVariables(Particles.VisibilityTag=$Particles.PartitionParticles.Partition) {isSetParameters=true}
   &InitializeParticle(
    Color=&MergeHelper_ColorSwitch(0=<rgba 1 0.8663 0.5041 0.72>, 1=<rgba 1 0.8663 0.5041 0.72>, 2=?, 3=?, Index=$Particles.PartitionParticles.Partition, Num=::NewEnumerator0)
    Position Offset=<0 -12 139.75>
    Sprite Size Mode=::NewEnumerator3
    Uniform Sprite Size=&MergeHelper_FloatSwitch(0=20, 1=18, 2=?, 3=?, Index=$Particles.PartitionParticles.Partition, Num=::NewEnumerator0)
    UsePositionOffset=true
    Write Material Random=true
   )
   &CameraOffset(Camera Offset Amount=&MergeHelper_FloatSwitch(0=0, 1=10, 2=?, 3=?, Index=$Particles.PartitionParticles.Partition, Num=::NewEnumerator0))
   &FN_Sprite_Attributes(
    Camera Offset=15
    Color=&MakeLinearColorFromVectorAndFloat("Vector (RGB)"=&MakeVectorFromLinearColorRGB(Color=$Particles.Color), "Float (Alpha)"=0.12)
    Create Camera Offset Attribute=true
    Create Color Attribute=true
    Create SpriteSize Attribute=true
    SpriteSize=&Multiply_Vector2D_ByFloat(Float=2.25, Vector2D=$Particles.SpriteSize)
    Write to Module Namespace=true
   )
   /SetVariables(Particles.HaloDynamic:Vector4f=&MakeVector4(X=1, Y=0, Z=1, W=0)) {isSetParameters=true}
  ]
  particleUpdateScript [
   &ParticleState(Kill Particles When Lifetime Has Elapsed=false)
   &ScaleSpriteSize(
    Scale Sprite Size Mode=::NewEnumerator0
    Uniform Scale Factor=&Multiply_Float(A=$System.EngineOwnerScale, B=&Lerp_Float(A=0.8, B=1.3, Alpha=$System.reactiveFloat))
   )
   &ScaleColor(
    Scale Alpha=&Multiply_Float(
     A=&FloatFromCurve(
      FloatCurve=DataInterface() {propertyValues {Curve {keys [
          {interpMode=RCIM_Cubic, leaveTangentWeight=0.1001, tangentMode=RCTM_User, time=0, value=0.9}
          {arriveTangentWeight=0.1021, interpMode=RCIM_Cubic, leaveTangentWeight=0.0679, tangentMode=RCTM_User, time=0.3003, value=1}
          {arriveTangentWeight=0.0666, interpMode=RCIM_Cubic, leaveTangentWeight=0.0667, tangentMode=RCTM_User, time=0.5, value=1}
          {arriveTangent=-0.2, interpMode=RCIM_Cubic, leaveTangent=-0.2, time=0.7, value=0.95}
          {arriveTangentWeight=0.1, interpMode=RCIM_Cubic, tangentMode=RCTM_User, time=1, value=0.9}
         ]}}}
      CurveIndex=Linked() {linkedVariable {name=NPC.FortniteNPC.RemappedTimeOfDay_0-1, type=NiagaraFloat}}
      Scale Curve=&MakeCustomFloatFromBool(Bool=$User.bIsFrontEndPreview, True Float=0.7, False Float=1)
     )
     B=&Lerp_Float(A=0.75, B=1, Alpha=$System.InterpolateOverTime.InterpolateOverTime_Float.Moving Average)
    )
    Scale RGB=&Multiply_VectorByFloat(
     Vector=<1 1 1>
     Float=&Lerp_Float(A=1, B=1.2, Alpha=$System.InterpolateOverTime.InterpolateOverTime_Float.Moving Average)
    )
    ScaleA=true
    ScaleRGB=true
   )
   /SetVariables(
    Particles.FN_Sprite_Attributes.SpriteSize:Vector2f=&Multiply_Vector2D_ByFloat(
     Float=&Multiply_Float(
      A=$System.EngineOwnerScale
      B=&Lerp_Float(A=0.8, B=1.3, Alpha=$System.InterpolateOverTime.InterpolateOverTime_Float.Moving Average)
     )
     Vector2D=&Multiply_Vector2D_ByFloat(
      Float=&Waveform("Waveform [1]"=::NewEnumerator0, "[1] Phase Scale"=2.5, "[1] Frequency"=1, "[1] Amplitude Min/Max"=<0.98 1>)
      Vector2D=$Particles.Initial.FN_Sprite_Attributes.SpriteSize
     )
    )
    Particles.HaloColor:LinearColor=&Multiply_LinearColorByFloat(
     Linear Color=$Particles.FN_Sprite_Attributes.Color
     Float=&Multiply_Float(
      A=&Waveform("Waveform [1]"=::NewEnumerator12, "[1] Frequency"=5, "[1] Amplitude Min/Max"=<0.85 1>)
      B=&Lerp_Float(A=1, B=1.2, Alpha=$System.InterpolateOverTime.InterpolateOverTime_Float.Moving Average)
     )
    )
   ) {isSetParameters=true}
  ]
  rendererClasses [&NiagaraSpriteRendererProperties, &NiagaraMeshRendererProperties]
  renderers [
   {rendererClass=&NiagaraSpriteRendererProperties}
   {rendererClass=&NiagaraSpriteRendererProperties, rendererIndex=1}
   {rendererClass=&NiagaraMeshRendererProperties, rendererIndex=2}
  ]
 }
 Flares_and_Sparkles {
  emitterUpdateScript [
   &EmitterState()
   &SpawnRate(
    Spawn Probability=0.8
    SpawnRate=&Multiply_Float(A=15, B=$System.ReactiveSpawnRate)
    Use Spawn Probability=true
   )
   &SpawnRate(
    Spawn Probability=0.8
    SpawnGroup=1
    SpawnRate=&Multiply_Float(A=3.5, B=$System.ReactiveSpawnRate)
    Use Spawn Probability=true
   ) {module=SpawnRate001}
  ]
  particleSpawnScript [
   /SetVariables(Particles.VisibilityTag=&SpawnGroup()) {isSetParameters=true, module=SetVariables_#10}
   &InitializeParticle(
    Color=&MergeHelper_ColorSwitch(
     0=&RandomRangeLinearColor(Minimum=<rgba 1 0.7676 0.4189 0.5>, Maximum=<rgba 1.185 1.25 0.5892 1>)
     1=&RandomRangeLinearColor(Minimum=<rgba 4 3.0703 1.6758 1>, Maximum=<rgba 4.7402 5 2.3568 1>)
     2=?
     3=?
     Index=&SpawnGroup()
     Num=::NewEnumerator0
    )
    Lifetime=&MergeHelper_FloatSwitch(
     0=&RandomRangeFloat(Minimum=0.65, Maximum=0.9)
     1=&RandomRangeFloat(Minimum=1.3, Maximum=1.5)
     2=?
     3=?
     Index=&SpawnGroup()
     Num=::NewEnumerator0
    )
    Sprite Size Mode=::NewEnumerator3
    Uniform Sprite Size=&Multiply_Float(
     A=&MergeHelper_FloatSwitch(
      0=&RandomRangeFloat(Minimum=0.2, Maximum=0.9)
      1=&RandomRangeFloat(Minimum=7.5, Maximum=10)
      2=?
      3=?
      Index=&SpawnGroup()
      Num=::NewEnumerator0
     )
     B=$System.EngineOwnerScale
    )
    UsePositionOffset=false
    Write Material Random=true
   )
   &ShapeLocation(
    MaskSpawnGroup=false
    Non Uniform Scale=<1 1 3>
    Offset=<0 0 68>
    Offset Mode=::NewEnumerator0
    Override Seed=false
    Sphere Radius=10
    Sphere Surface Distribution=0.8643
    UseOwnerScale=true
   )
   &CameraOffset(Camera Offset Amount=15)
   &PartitionParticles()
  ]
  particleUpdateScript [
   &ParticleState()
   /SetVariables(Particles.FlickerWaveform:NiagaraFloat=&Waveform(
     Global Phase=&Add_Float(A=$Engine.Time, B=&RandomRangeFloat(Minimum=0, Maximum=1))
     "Waveform [1]"=::NewEnumerator10
     "[1] Frequency"=10
     "[1] Amplitude Min/Max"=<0.5 1.4>
    )) {isSetParameters=true, module=SetVariables_#11}
   &ScaleSpriteSize(
    Scale Sprite Size Mode=::NewEnumerator0
    Uniform Scale Factor=&MergeHelper_FloatSwitch(
     0=&Multiply_Float(
      A=&FloatFromCurve(
       FloatCurve=DataInterface() {propertyValues {Curve {
          defaultValue=0
          keys [
           {arriveTangent=9.0725, interpMode=RCIM_Cubic, leaveTangent=9.0725, tangentMode=RCTM_User, time=0, value=0}
           {
            arriveTangent=-1.1669
            arriveTangentWeight=0.1274
            interpMode=RCIM_Cubic
            leaveTangent=-1.1669
            leaveTangentWeight=0.7778
            tangentMode=RCTM_User
            time=0.1408
            value=0.9847
           }
           {arriveTangent=-0.0405, interpMode=RCIM_Cubic, leaveTangent=-0.0405, tangentMode=RCTM_User, time=1, value=0}
          ]
         }}}
       Scale Curve=$Particles.FlickerWaveform
      )
      B=&Lerp_Float(A=0.5, B=1.3, Alpha=$System.reactiveFloat)
     )
     1=&Multiply_Float(
      A=&FloatFromCurve(FloatCurve=DataInterface() {propertyValues {Curve {
          defaultValue=0
          keys [
           {interpMode=RCIM_Cubic, leaveTangentWeight=0.6384, tangentMode=RCTM_User, time=0, value=0}
           {arriveTangentWeight=0.0333, interpMode=RCIM_Cubic, leaveTangentWeight=0.2, tangentMode=RCTM_User, time=0.5, value=1}
           {arriveTangentWeight=0.2839, interpMode=RCIM_Cubic, tangentMode=RCTM_User, time=1, value=0}
          ]
         }}})
      B=&Lerp_Float(A=0.6, B=1.75, Alpha=$System.reactiveFloat)
     )
     2=?
     3=?
     Index=&SpawnGroup()
     Num=::NewEnumerator0
    )
   )
   &CurlNoiseForce(Bias Noise=false, Noise Strength=&RandomRangeFloat(Minimum=10, Maximum=15), Pan Noise=false)
   &VortexForce(Origin Pull Amount=25, UseOriginOffset=true, Vortex Force Amount=-100)
   &SolveForcesAndVelocity(Clamp Velocity=false, Limit Acceleration=false)
   &ScaleColor(
    Scale Alpha=&Multiply_Float(
     A=&MergeHelper_FloatSwitch(
      0=&Multiply_Float(
       A=&FloatFromCurve(
        FloatCurve=DataInterface() {propertyValues {Curve {
           defaultValue=1
           keys [
            {arriveTangent=5.5696, interpMode=RCIM_Cubic, leaveTangent=5.5696, tangentMode=RCTM_User, time=0, value=0}
            {arriveTangent=-0.8557, interpMode=RCIM_Cubic, leaveTangent=-0.8557, tangentMode=RCTM_User, time=0.1006, value=1}
            {tangentMode=RCTM_User, time=1, value=0}
           ]
          }}}
        Scale Curve=$Particles.FlickerWaveform
       )
       B=&RandomRangeFloat(Minimum=0.5, Maximum=1)
      )
      1=&FloatFromCurve(
       FloatCurve=DataInterface() {propertyValues {Curve {
          defaultValue=1
          keys [
           {arriveTangent=4.6847, interpMode=RCIM_Cubic, leaveTangent=4.6847, leaveTangentWeight=0.4012, tangentMode=RCTM_User, time=0, value=0}
           {arriveTangentWeight=0.0442, interpMode=RCIM_Cubic, leaveTangentWeight=0.3945, tangentMode=RCTM_User, time=0.2732, value=1}
           {arriveTangentWeight=0.1189, interpMode=RCIM_Cubic, leaveTangentWeight=0.1652, tangentMode=RCTM_User, time=0.5045, value=1}
           {arriveTangent=-2.3008, tangentMode=RCTM_User, time=1, value=0}
          ]
         }}}
       Scale Curve=&RandomRangeFloat(Minimum=0.1, Maximum=0.5)
      )
      2=?
      3=?
      Index=&SpawnGroup()
      Num=::NewEnumerator0
     )
     B=&Lerp_Float(A=0.8, B=1, Alpha=$System.InterpolateOverTime.InterpolateOverTime_Float.Moving Average)
    )
    Scale RGB=&Multiply_VectorByFloat(
     Vector=&MergeHelper_VectorSwitch(
      0=&VectorFromFloat(Value=&FloatFromCurve(
        FloatCurve=DataInterface() {propertyValues {Curve {
           defaultValue=1
           keys [
            {arriveTangent=9.0725, interpMode=RCIM_Cubic, leaveTangent=9.0725, tangentMode=RCTM_User, time=0, value=1}
            {arriveTangent=-0.7324, interpMode=RCIM_Cubic, leaveTangent=-0.7324, tangentMode=RCTM_User, time=0.1363, value=1.8121}
            {arriveTangent=-0.0405, interpMode=RCIM_Cubic, leaveTangent=-0.0405, tangentMode=RCTM_User, time=1, value=1}
           ]
          }}}
        Scale Curve=$Particles.FlickerWaveform
       ))
      1=&VectorFromFloat(Value=1)
      2=?
      3=?
      Index=&SpawnGroup()
      Num=::NewEnumerator0
     )
     Float=&Lerp_Float(A=1, B=1.1, Alpha=$System.InterpolateOverTime.InterpolateOverTime_Float.Moving Average)
    )
    ScaleA=true
    ScaleRGB=true
   )
   /SetVariables(
    Particles.Glow Color=&MergeHelper_ColorSwitch(
     0=&MakeLinearColorFromVectorAndFloat(
      "Vector (RGB)"=&MakeVectorFromLinearColorRGB(Color=&MultiplyLinearColor(Linear Color=$Particles.Color, Scale Factor=&MakeVector4_FromLinearColor(LinearColor=$Particles.Color)))
      "Float (Alpha)"=0.08
     )
     1=&MakeLinearColorFromVectorAndFloat(
      "Vector (RGB)"=&MakeVectorFromLinearColorRGB(Color=&MultiplyLinearColor(
        Linear Color=$Particles.Color
        Scale Factor=&MakeVector4_FromLinearColor(LinearColor=<rgba 0.15 0.0919 0.0339 1>)
       ))
      "Float (Alpha)"=0.5
     )
     2=?
     3=?
     Index=&SpawnGroup()
     Num=::NewEnumerator0
    )
    Particles.Glow Size=&Multiply_Vector2D(
     A=$Particles.SpriteSize
     B=&MergeHelper_Vector2Switch(
      0=<3.5 4.5>
      1=&Multiply_Vector2D_ByFloat(Float=&Lerp_Float(A=0.8, B=1.5, Alpha=$System.reactiveFloat), Vector2D=<0.5 0.5>)
      2=?
      3=?
      4=?
      Index=&SpawnGroup()
      Num=::NewEnumerator0
     )
    )
   ) {isSetParameters=true, module=SetVariables_#12}
   /SetVariables(
    Particles.OffsetPosition:NiagaraPosition=&AddVectorToPosition(
     Position=$Particles.Position
     Vector=&Lerp_Vector(
      A=<0 0 6>
      B=<0 0 -2>
      Alpha=&MakeFloatFromLinearColor(
       LinearColor=$System.InterpolateOverTime001.InterpolateOverTime_LinearColor.Moving Average
       Channel:ENiagaraFloatFromLinearColorOptions=::NewEnumerator3
      )
     )
    )
    Particles.OffsetPosition2:NiagaraPosition=&AddVectorToPosition(
     Position=$Particles.Position
     Vector=&Lerp_Vector(
      A=<0 0 -2>
      B=<0 0 4>
      Alpha=&MakeFloatFromLinearColor(
       LinearColor=$System.InterpolateOverTime001.InterpolateOverTime_LinearColor.Moving Average
       Channel:ENiagaraFloatFromLinearColorOptions=::NewEnumerator3
      )
     )
    )
   ) {isSetParameters=true, module=SetVariables_#13}
  ]
  rendererClasses [&NiagaraSpriteRendererProperties]
  renderers [
   {rendererClass=&NiagaraSpriteRendererProperties}
   {rendererClass=&NiagaraSpriteRendererProperties, rendererIndex=1}
   {rendererClass=&NiagaraSpriteRendererProperties, rendererIndex=2}
   {rendererClass=&NiagaraSpriteRendererProperties, rendererIndex=3}
  ]
 }
 NE_SpiralRibbon {
  emitterSpawnScript [/SetVariables(Emitter.ParticlePerRibbon=12, Emitter.RibbonNumber=1) {isSetParameters=true}]
  emitterUpdateScript [
   &EmitterState(
    Inactive Response=::NewEnumerator1
    Life Cycle Mode=::NewEnumerator1
    Loop Behavior=::NewEnumerator1
    Loop Duration Mode=::NewEnumerator1
    UseLoopDelay=false
   )
   &SpawnBurst_Instantaneous(
    Spawn Count=&Multiply_Int(A=$Emitter.ParticlePerRibbon, B=$Emitter.RibbonNumber)
    Use Spawn Probability=false
    UseLoopCountLimit=true
   )
  ]
  particleSpawnScript [
   &InitializeParticle(Lifetime=5, Ribbon Width Mode=::NewEnumerator1, Sprite Size Mode=::NewEnumerator3, UsePositionOffset=false, Write Material Random=true)
   &PartitionParticles(
    ID to Partition:NiagaraInt32=&ReturnExecIndex()
    Number Of Partitions=$Emitter.RibbonNumber
    Particles Per Partition:NiagaraInt32=$Emitter.ParticlePerRibbon
    Partition Mode:ENiagara_PartitionMode=::NewEnumerator1
   )
   /SetVariables(Particles.NormalizedRibbonPos:NiagaraFloat=&OneMinusFloat(Float=$Particles.PartitionParticles.Normalized Index In Partition)) {isSetParameters=true}
   /SetVariables(Particles.SpiralFreq=0.6, Particles.SpiralStrength=1) {isSetParameters=true}
   /SetVariables(Particles.RibbonU0Override=&OneMinusFloat(Float=$Particles.NormalizedRibbonPos)) {isSetParameters=true}
   /SetVariables() {isSetParameters=true}
  ]
  particleUpdateScript [
   &ParticleState(Kill Particles When Lifetime Has Elapsed=false)
   /SetVariables(Particles.GoalPos:NiagaraPosition=&AddVectorToPosition(
     Position=$Particles.SkeletalMeshLocation.SampledPosition
     Vector=&Lerp_Vector(
      A=<0 0 35>
      B=&Multiply_Vector(
       A=<1 1 100>
       B=&MakeVector(
        X=1
        Y=1
        Z=&Add_Float(A=1, B=&Multiply_Float(A=$System.InterpolateOverTime.InterpolateOverTime_Float.Moving Average, B=0.08))
       )
      )
      Alpha=$Particles.NormalizedRibbonPos
     )
    )) {isSetParameters=true}
   /SetVariables(Particles.SpiralStrength:NiagaraFloat=&Multiply_Float(
     A=&Multiply_Float(
      A=&Multiply_Float(A=2, B=1)
      B=&Waveform(
       Global Phase=&Frac_Float(Float=&Add_Float(A=$Engine.Time, B=$Particles.NormalizedRibbonPos))
       "Waveform [1]"=::NewEnumerator12
       "[1] Frequency"=0.25
       "[1] Amplitude Min/Max"=<0.75 1.25>
      )
     )
     B=&FloatFromCurve(
      FloatCurve=DataInterface() {propertyValues {Curve {
         defaultValue=0
         keys [<at 0 0.85 RCIM_Cubic>, {interpMode=RCIM_Cubic, tangentMode=RCTM_User, time=0.4985, value=1}, <at 0.9968 0.85 RCIM_Cubic>]
        }}}
      CurveIndex=$Particles.NormalizedRibbonPos
      Scale Curve=2
     )
    )) {isSetParameters=true}
   /SetVariables(Particles.SpiralTime:NiagaraFloat=&Add_Float(
     A=&Add_Float(A=&Multiply_Float(A=$System.Age, B=0.15), B=$Particles.NormalizedRibbonPos)
     B=&Multiply_FloatByInt(
      Float=&Divide_Float(A=$Particles.SpiralFreq, B=&MakeFloatFromInt(Int=$Emitter.RibbonNumber))
      Integer=$Particles.PartitionParticles.Partition
     )
    )) {isSetParameters=true}
   /SetVariables(Particles.SpiralOffset:Vector3f=&MakeVector(
     X=&Sine(
      Normalized Angle=$Particles.SpiralTime
      Period=$Particles.SpiralFreq
      Scale=&Multiply_Float(A=$Particles.SpiralStrength, B=2)
      Bias=0
     )
     Y=&Cosine(
      Normalized Angle=$Particles.SpiralTime
      Period=$Particles.SpiralFreq
      Scale=&Multiply_Float(A=$Particles.SpiralStrength, B=2)
      Bias=0
     )
     Z=0
    )) {isSetParameters=true}
   /SetVariables(Particles.Angle Fade:NiagaraFloat=&Lerp_Float(
     A=0.6
     B=1
     Alpha=&FloatFromCurve(
      FloatCurve=DataInterface() {propertyValues {Curve {keys [<at 0.5 1 RCIM_Cubic>,<at 0.9 0 RCIM_Cubic>]}}}
      CurveIndex=&Abs_Float(Float:NiagaraFloat=&DotProduct(Vector A=&ReturnCameraVector(Camera Vectors=::NewEnumerator1, Coordinate Space=::Local), Vector B=<0 0 1>))
     )
    )) {isSetParameters=true}
   &InheritSourceMovement(
    Applied Position Scale=&VectorFromFloat(Value=&Multiply_Float(
      A=-0.225
      B=&FloatFromCurve(
       FloatCurve=DataInterface() {propertyValues {Curve {keys [
           {arriveTangent=0.3323, interpMode=RCIM_Cubic, leaveTangent=0.3323, tangentMode=RCTM_User, time=0, value=0}
           <at 0.5 1 RCIM_Cubic>
           {arriveTangent=0.1272, interpMode=RCIM_Cubic, leaveTangent=0.1272, tangentMode=RCTM_User, time=1, value=0}
          ]}}}
       CurveIndex=$Particles.NormalizedRibbonPos
      )
     ))
    ApplyForce=false
    ApplyPosition=true
    ApplyPositionFalloff=false
    Source Velocity:Vector3f=&TransformVector(Vector=$Engine.Owner.Velocity, Source Space=::World, Destination Space=::Local)
   )
   &SolveForcesAndVelocity(Clamp Velocity=false, Limit Acceleration=false)
   &ScaleRibbonWidth(Ribbon Width Scale=&Multiply_Float(
     A=&Multiply_Float(
      A=&FloatFromCurve(
       FloatCurve=DataInterface() {propertyValues {Curve {keys [
           {arriveTangent=0.6169, interpMode=RCIM_Cubic, leaveTangent=0.6169, tangentMode=RCTM_User, time=-0.0018, value=0.6085}
           {arriveTangentWeight=0.107, interpMode=RCIM_Cubic, leaveTangentWeight=0.2149, tangentMode=RCTM_User, time=0.2651, value=1.0492}
           {arriveTangent=-0.0917, interpMode=RCIM_Cubic, leaveTangent=-0.0917, time=0.5, value=0.6837}
           {arriveTangentWeight=0.2039, interpMode=RCIM_Cubic, leaveTangentWeight=0.0761, tangentMode=RCTM_User, time=0.8012, value=1}
           {arriveTangent=-0.6127, interpMode=RCIM_Cubic, leaveTangent=-0.6127, tangentMode=RCTM_User, time=1.0011, value=0.6331}
          ]}}}
       CurveIndex=$Particles.NormalizedRibbonPos
       Scale Curve=&Multiply_Float(A=1.3, B=$System.EngineOwnerScale)
      )
      B=&FloatFromCurve(
       FloatCurve=DataInterface() {propertyValues {Curve {keys [
           {interpMode=RCIM_Cubic, leaveTangentWeight=0.3904, tangentMode=RCTM_User, time=0, value=0.2}
           {arriveTangentWeight=0.2395, interpMode=RCIM_Cubic, leaveTangentWeight=0.1617, tangentMode=RCTM_User, time=0.4, value=1}
           {arriveTangentWeight=0.1749, interpMode=RCIM_Cubic, leaveTangentWeight=0.2138, tangentMode=RCTM_User, time=0.6, value=1}
           {arriveTangentWeight=0.1333, interpMode=RCIM_Cubic, tangentMode=RCTM_User, time=1, value=0.2}
          ]}}}
       CurveIndex=$Particles.NormalizedRibbonPos
      )
     )
     B=&Lerp_Float(A=0.9, B=1.25, Alpha=$System.InterpolateOverTime.InterpolateOverTime_Float.Moving Average)
    ))
   /SetVariables(Particles.Position=&AddVectorToPosition(Position=$Particles.GoalPos, Vector=$Particles.SpiralOffset)) {isSetParameters=true}
   /SetVariables(Particles.RibbonFacing=&NormalizeVector(
     Vector To Normalize=&TransformVector(Vector=$Particles.SpiralOffset, Destination Space=::Simulation)
     Safe Normalize=true
     Safe Fallback Threshold=0
     Safe Fallback Vector=<0 0 0.001>
    )) {isSetParameters=true}
   /SetVariables(Particles.BackSpiralColor:LinearColor=&MultiplyLinearColor(
     Linear Color=$Particles.Color
     Scale Factor=&MakeVector4_FromLinearColor(LinearColor=<rgba 0.1224 0.5 0.5 1>)
    )) {enabled=false, isSetParameters=true}
   /SetVariables(Particles.BackRibbonFacing:Vector3f=&NormalizeVector(
     Vector To Normalize=&TransformVector(Vector=&Multiply_VectorByFloat(Vector=$Particles.SpiralOffset, Float=-1), Destination Space=::Simulation)
     Safe Normalize=true
     Safe Fallback Threshold=0
     Safe Fallback Vector=<0 0 0.001>
    )) {enabled=false, isSetParameters=true}
   /SetVariables(
    Particles.GlowWidth:NiagaraFloat=&Multiply_Float(A=$Particles.RibbonWidth, B=3.2)
    Particles.ShimmerWidth:NiagaraFloat=&Multiply_Float(A=$Particles.RibbonWidth, B=1.5)
   ) {isSetParameters=true}
   /SetVariables(
    Particles.Color_Lines:LinearColor=&MakeLinearColorFromVectorAndFloat(
     "Vector (RGB)"=&Multiply_VectorByFloat(
      Vector=&MakeVectorFromLinearColorRGB(Color=&ColorFromCurve(
        ColorCurve=DataInterface() {propertyValues {
          AlphaCurve {keys [<at 0.4807 1 RCIM_Cubic>]}
          BlueCurve {keys [<at 0 0.2346 RCIM_Cubic>,<at 0.4815 0.7474>,<at 1 0.2346 RCIM_Cubic>]}
          GreenCurve {keys [<at 0 0.5089 RCIM_Cubic>,<at 0.4815 0.8869>,<at 1 0.5089 RCIM_Cubic>]}
          RedCurve {keys [<at 0 0.6939 RCIM_Cubic>,<at 0.4815 1>,<at 1 0.6939 RCIM_Cubic>]}
         }}
        CurveIndex=$Particles.NormalizedRibbonPos
       ))
      Float=&Lerp_Float(A=1, B=1.75, Alpha=$System.InterpolateOverTime.InterpolateOverTime_Float.Moving Average)
     )
     "Float (Alpha)"=&Multiply_Float(
      A=&Multiply_Float(
       A=&FloatFromCurve(
        FloatCurve=DataInterface() {propertyValues {Curve {
           defaultValue=1
           keys [
            {interpMode=RCIM_Cubic, tangentMode=RCTM_User, time=0, value=0}
            <at 0.25 1 RCIM_Cubic>
            {interpMode=RCIM_Cubic, tangentMode=RCTM_User, time=0.5, value=0}
            <at 0.75 1 RCIM_Cubic>
            {tangentMode=RCTM_User, time=1, value=0}
           ]
          }}}
        CurveIndex=&Frac_Float(Float=&Add_Float(A=&Multiply_Float(A=$System.Age, B=0.15), B=$Particles.NormalizedRibbonPos))
       )
       B=&FloatFromCurve(
        FloatCurve=DataInterface() {propertyValues {Curve {
           defaultValue=1
           keys [
            {interpMode=RCIM_Cubic, leaveTangentWeight=0.363, tangentMode=RCTM_User, time=0, value=0}
            {interpMode=RCIM_Cubic, tangentMode=RCTM_User, time=0.5, value=1}
            {arriveTangentWeight=0.2776, interpMode=RCIM_Cubic, tangentMode=RCTM_User, time=1, value=0}
           ]
          }}}
        CurveIndex=$Particles.NormalizedRibbonPos
        Scale Curve=$Particles.Angle Fade
       )
      )
      B=&Lerp_Float(A=1, B=1.2, Alpha=$System.InterpolateOverTime.InterpolateOverTime_Float.Moving Average)
     )
    )
    Particles.LinesWidth:NiagaraFloat=&Multiply_Float(A=$Particles.RibbonWidth, B=1.7)
   ) {isSetParameters=true}
   /SetVariables(Particles.Color_Shimmer:LinearColor=&MakeLinearColorFromVectorAndFloat(
     "Vector (RGB)"=&Multiply_VectorByFloat(
      Vector=&MakeVectorFromLinearColorRGB(Color=<rgba 40 31.9472 22.1911 1>)
      Float=&Lerp_Float(A=0.9, B=5, Alpha=$System.InterpolateOverTime.InterpolateOverTime_Float.Moving Average)
     )
     "Float (Alpha)"=$Particles.Angle Fade
    )) {isSetParameters=true}
   /SetVariables(Particles.Color_GlowRibbon:LinearColor=&MakeLinearColorFromVectorAndFloat(
     "Vector (RGB)"=&Multiply_VectorByFloat(
      Vector=&MakeVectorFromLinearColorRGB(Color=&ColorFromCurve(
        ColorCurve=DataInterface() {propertyValues {
          AlphaCurve {keys [<at 0.4807 1 RCIM_Cubic>]}
          BlueCurve {keys [<at 0.0334 0.2016 RCIM_Cubic>,<at 0.4815 0.5938>,<at 0.9644 0.2016 RCIM_Cubic>]}
          GreenCurve {keys [<at 0.0334 0.4969 RCIM_Cubic>,<at 0.4815 0.9377>,<at 0.9644 0.4969 RCIM_Cubic>]}
          RedCurve {keys [<at 0.0334 0.6939 RCIM_Cubic>,<at 0.4815 1>,<at 0.9644 0.6939 RCIM_Cubic>]}
         }}
        CurveIndex=$Particles.NormalizedRibbonPos
       ))
      Float=&Lerp_Float(A=0.9, B=2.5, Alpha=$System.InterpolateOverTime.InterpolateOverTime_Float.Moving Average)
     )
     "Float (Alpha)"=&FloatFromCurve(
      FloatCurve=DataInterface() {propertyValues {Curve {keys [
          {arriveTangent=0.6799, interpMode=RCIM_Cubic, leaveTangent=0.6799, tangentMode=RCTM_User, time=0, value=0}
          <at 0.4666 1 RCIM_Cubic>
          {arriveTangent=-0.2647, interpMode=RCIM_Cubic, leaveTangent=-0.2647, tangentMode=RCTM_User, time=1, value=0}
         ]}}}
      CurveIndex=$Particles.NormalizedRibbonPos
      Scale Curve=&Multiply_Float(
       A=&FloatFromCurve(
        FloatCurve=DataInterface() {propertyValues {Curve {keys [<at 0 0.5>,<at 0.2854 1>,<at 0.6409 1>,<at 0.912 0.5>]}}}
        CurveIndex=Linked() {linkedVariable {name=NPC.FortniteNPC.TimeOfDay_0-1, type=NiagaraFloat}}
       )
       B=&Multiply_Float(
        A=&MakeCustomFloatFromBool(Bool=$User.bIsFrontEndPreview, True Float=0.5, False Float=1)
        B=&Multiply_Float(A=0.4, B=&Lerp_Float(A=1, B=1, Alpha=$System.InterpolateOverTime.InterpolateOverTime_Float.Moving Average))
       )
      )
     )
    )) {isSetParameters=true}
   &ScaleColor(
    Scale Alpha=&Multiply_Float(
     A=&MakeCustomFloatFromBool(Bool=$User.bIsFrontEndPreview, True Float=0.75, False Float=1)
     B=&Lerp_Float(
      A=&Lerp_Float(A=1, B=1.5, Alpha=$NPC.FortniteNPC.EventAudio_Amplitude)
      B=&Lerp_Float(A=1, B=1.5, Alpha=$System.FEFakeReactivity)
      Alpha=$System.UsePreview
     )
    )
    Scale RGB=&Multiply_VectorByFloat(
     Vector=&MakeVectorFromLinearColorRGB(Color=&ColorFromCurve(
       ColorCurve=DataInterface() {propertyValues {
         AlphaCurve {keys []}
         BlueCurve {keys [<at 0.0939 0.7>, {arriveTangent=0.3339, interpMode=RCIM_Cubic, leaveTangent=0.3339, time=0.7241, value=1.25}, <at 0.9923 1>]}
         GreenCurve {keys [<at 0.0939 0.3162>, {arriveTangent=0.6329, interpMode=RCIM_Cubic, leaveTangent=0.6329, time=0.7241, value=1.106}, <at 0.9923 0.8848>]}
         RedCurve {keys [<at 0.0939 0.0164>, {arriveTangent=0.3591, interpMode=RCIM_Cubic, leaveTangent=0.3591, time=0.7241, value=0.4238}, <at 0.9923 0.339>]}
        }}
       CurveIndex=$Particles.NormalizedRibbonPos
      ))
     Float=&Waveform("Waveform [1]"=::NewEnumerator0, "[1] Phase Scale"=2, "[1] Frequency"=0.3, "[1] Amplitude Min/Max"=<40 50>)
    )
    ScaleA=true
    ScaleRGB=true
   ) {enabled=false, module=ScaleColor003}
  ]
  rendererClasses [&NiagaraRibbonRendererProperties]
  renderers [
   {rendererClass=&NiagaraRibbonRendererProperties}
   {rendererClass=&NiagaraRibbonRendererProperties, rendererIndex=1}
   {rendererClass=&NiagaraRibbonRendererProperties, rendererIndex=2}
   {rendererClass=&NiagaraRibbonRendererProperties, rendererIndex=3}
  ]
 }
 Stars_and_Motes001 {
  bEnabled=false
  emitterUpdateScript [
   &EmitterState()
   &SpawnBurst_Instantaneous(Spawn Count=7, Use Spawn Probability=false, UseLoopCountLimit=true) {enabled=false}
   &SpawnRate(
    Spawn Probability=0.8
    SpawnRate=&Multiply_Float(A=15, B=$System.ReactiveSpawnRate)
    Use Spawn Probability=true
   )
   &SpawnBurst_Instantaneous(Spawn Count=2, Spawn Group=1, Use Spawn Probability=false, UseLoopCountLimit=true) {enabled=false, module=SpawnBurst_Instantaneous001}
   &SpawnRate(
    Spawn Probability=0.8
    SpawnGroup=1
    SpawnRate=&Multiply_Float(A=2.5, B=$System.ReactiveSpawnRate)
    Use Spawn Probability=true
   ) {module=SpawnRate001}
  ]
  particleSpawnScript [
   /SetVariables(Particles.VisibilityTag=&SpawnGroup()) {isSetParameters=true, module=SetVariables_#10}
   &InitializeParticle(
    Color=&MergeHelper_ColorSwitch(
     0=&RandomRangeLinearColor(Minimum=<rgba 1 0.7676 0.4189 0.5>, Maximum=<rgba 1.185 1.25 0.5892 1>)
     1=&RandomRangeLinearColor(Minimum=<rgba 1 0.8125 0.5312 0.7>, Maximum=<rgba 0.959 1 0.5833 1>)
     2=?
     3=?
     Index=&SpawnGroup()
     Num=::NewEnumerator0
    )
    Lifetime=&MergeHelper_FloatSwitch(
     0=&RandomRangeFloat(Minimum=0.65, Maximum=0.9)
     1=&RandomRangeFloat(Minimum=1.3, Maximum=1.5)
     2=?
     3=?
     Index=&SpawnGroup()
     Num=::NewEnumerator0
    )
    Sprite Size Mode=::NewEnumerator3
    Uniform Sprite Size=&Multiply_Float(
     A=&MergeHelper_FloatSwitch(
      0=&RandomRangeFloat(Minimum=0.2, Maximum=0.9)
      1=&RandomRangeFloat(Minimum=3.5, Maximum=5)
      2=?
      3=?
      Index=&SpawnGroup()
      Num=::NewEnumerator0
     )
     B=$System.EngineOwnerScale
    )
    UsePositionOffset=false
    Write Material Random=true
   )
   &ShapeLocation(
    MaskSpawnGroup=false
    Non Uniform Scale=<1 1 3>
    Offset=<0 0 68>
    Offset Mode=::NewEnumerator0
    Override Seed=false
    Sphere Radius=10
    Sphere Surface Distribution=0.8643
    UseOwnerScale=true
   )
   &CameraOffset(Camera Offset Amount=15)
   &PartitionParticles()
  ]
  particleUpdateScript [
   &ParticleState()
   /SetVariables(Particles.FlickerWaveform:NiagaraFloat=&Waveform(
     Global Phase=&Add_Float(A=$Engine.Time, B=&RandomRangeFloat(Minimum=0, Maximum=1))
     "Waveform [1]"=::NewEnumerator10
     "[1] Frequency"=10
     "[1] Amplitude Min/Max"=<0.5 1.4>
    )) {isSetParameters=true, module=SetVariables_#11}
   &ScaleSpriteSize(
    Scale Sprite Size Mode=::NewEnumerator0
    Uniform Scale Factor=&MergeHelper_FloatSwitch(
     0=&Multiply_Float(
      A=&FloatFromCurve(
       FloatCurve=DataInterface() {propertyValues {Curve {
          defaultValue=0
          keys [
           {arriveTangent=9.0725, interpMode=RCIM_Cubic, leaveTangent=9.0725, tangentMode=RCTM_User, time=0, value=0}
           {
            arriveTangent=-1.1669
            arriveTangentWeight=0.1274
            interpMode=RCIM_Cubic
            leaveTangent=-1.1669
            leaveTangentWeight=0.7778
            tangentMode=RCTM_User
            time=0.1408
            value=0.9847
           }
           {arriveTangent=-0.0405, interpMode=RCIM_Cubic, leaveTangent=-0.0405, tangentMode=RCTM_User, time=1, value=0}
          ]
         }}}
       Scale Curve=$Particles.FlickerWaveform
      )
      B=&Lerp_Float(A=0.5, B=1.3, Alpha=$System.reactiveFloat)
     )
     1=&Multiply_Float(
      A=&FloatFromCurve(FloatCurve=DataInterface() {propertyValues {Curve {
          defaultValue=0
          keys [
           {arriveTangent=5.1744, interpMode=RCIM_Cubic, leaveTangent=5.1744, tangentMode=RCTM_User, time=0, value=0}
           {arriveTangentWeight=0.0333, interpMode=RCIM_Cubic, leaveTangentWeight=0.2, tangentMode=RCTM_User, time=0.1679, value=0.9847}
           {arriveTangentWeight=0.2001, interpMode=RCIM_Cubic, leaveTangentWeight=0.2809, tangentMode=RCTM_Break, time=0.55, value=1}
           {arriveTangentWeight=0.2839, interpMode=RCIM_Cubic, tangentMode=RCTM_User, time=1, value=0}
          ]
         }}})
      B=&Lerp_Float(A=0.4, B=1.75, Alpha=$System.reactiveFloat)
     )
     2=?
     3=?
     Index=&SpawnGroup()
     Num=::NewEnumerator0
    )
   )
   &CurlNoiseForce(Bias Noise=false, Noise Strength=&RandomRangeFloat(Minimum=10, Maximum=15), Pan Noise=false)
   &VortexForce(Origin Pull Amount=25, UseOriginOffset=true, Vortex Force Amount=-100)
   &SolveForcesAndVelocity(Clamp Velocity=false, Limit Acceleration=false)
   &ScaleColor(
    Scale Alpha=&Multiply_Float(
     A=&MergeHelper_FloatSwitch(
      0=&FloatFromCurve(
       FloatCurve=DataInterface() {propertyValues {Curve {
          defaultValue=1
          keys [
           {arriveTangent=5.5696, interpMode=RCIM_Cubic, leaveTangent=5.5696, tangentMode=RCTM_User, time=0, value=0}
           {arriveTangent=-0.8557, interpMode=RCIM_Cubic, leaveTangent=-0.8557, tangentMode=RCTM_User, time=0.1006, value=1}
           {tangentMode=RCTM_User, time=1, value=0}
          ]
         }}}
       Scale Curve=$Particles.FlickerWaveform
      )
      1=&FloatFromCurve(FloatCurve=DataInterface() {propertyValues {Curve {
          defaultValue=1
          keys [
           {interpMode=RCIM_Cubic, leaveTangentWeight=0.4012, tangentMode=RCTM_User, time=0, value=0}
           {arriveTangentWeight=0.0442, interpMode=RCIM_Cubic, leaveTangentWeight=0.3945, tangentMode=RCTM_User, time=0.1477, value=1}
           {arriveTangentWeight=0.1189, interpMode=RCIM_Cubic, leaveTangentWeight=0.1652, tangentMode=RCTM_User, time=0.5045, value=1}
           {arriveTangent=-2.3008, tangentMode=RCTM_User, time=1, value=0}
          ]
         }}})
      2=?
      3=?
      Index=&SpawnGroup()
      Num=::NewEnumerator0
     )
     B=&Lerp_Float(A=0.8, B=1, Alpha=$System.InterpolateOverTime.InterpolateOverTime_Float.Moving Average)
    )
    Scale RGB=&Multiply_VectorByFloat(
     Vector=&MergeHelper_VectorSwitch(
      0=&VectorFromFloat(Value=&FloatFromCurve(
        FloatCurve=DataInterface() {propertyValues {Curve {
           defaultValue=1
           keys [
            {arriveTangent=9.0725, interpMode=RCIM_Cubic, leaveTangent=9.0725, tangentMode=RCTM_User, time=0, value=1}
            {arriveTangent=-0.7324, interpMode=RCIM_Cubic, leaveTangent=-0.7324, tangentMode=RCTM_User, time=0.1363, value=1.8121}
            {arriveTangent=-0.0405, interpMode=RCIM_Cubic, leaveTangent=-0.0405, tangentMode=RCTM_User, time=1, value=1}
           ]
          }}}
        Scale Curve=$Particles.FlickerWaveform
       ))
      1=&VectorFromFloat(Value=1)
      2=?
      3=?
      Index=&SpawnGroup()
      Num=::NewEnumerator0
     )
     Float=&Lerp_Float(A=1, B=1.1, Alpha=$System.InterpolateOverTime.InterpolateOverTime_Float.Moving Average)
    )
    ScaleA=true
    ScaleRGB=true
   )
   /SetVariables(
    Particles.Glow Color=&MergeHelper_ColorSwitch(
     0=&MakeLinearColorFromVectorAndFloat(
      "Vector (RGB)"=&MakeVectorFromLinearColorRGB(Color=&MultiplyLinearColor(Linear Color=$Particles.Color, Scale Factor=&MakeVector4_FromLinearColor(LinearColor=$Particles.Color)))
      "Float (Alpha)"=0.08
     )
     1=&MakeLinearColorFromVectorAndFloat(
      "Vector (RGB)"=&MakeVectorFromLinearColorRGB(Color=&MultiplyLinearColor(Linear Color=$Particles.Color, Scale Factor=&MakeVector4_FromLinearColor(LinearColor=$Particles.Color)))
      "Float (Alpha)"=0.3
     )
     2=?
     3=?
     Index=&SpawnGroup()
     Num=::NewEnumerator0
    )
    Particles.Glow Size=&Multiply_Vector2D(
     A=$Particles.SpriteSize
     B=&MergeHelper_Vector2Switch(0=<3.5 4.5>, 1=<1.75 1.75>, 2=?, 3=?, 4=?, Index=&SpawnGroup(), Num=::NewEnumerator0)
    )
   ) {isSetParameters=true, module=SetVariables_#12}
   /SetVariables(
    Particles.OffsetPosition:NiagaraPosition=&AddVectorToPosition(
     Position=$Particles.Position
     Vector=&Lerp_Vector(
      A=<0 0 6>
      B=<0 0 -2>
      Alpha=&MakeFloatFromLinearColor(
       LinearColor=$System.InterpolateOverTime001.InterpolateOverTime_LinearColor.Moving Average
       Channel:ENiagaraFloatFromLinearColorOptions=::NewEnumerator3
      )
     )
    )
    Particles.OffsetPosition2:NiagaraPosition=&AddVectorToPosition(
     Position=$Particles.Position
     Vector=&Lerp_Vector(
      A=<0 0 -2>
      B=<0 0 4>
      Alpha=&MakeFloatFromLinearColor(
       LinearColor=$System.InterpolateOverTime001.InterpolateOverTime_LinearColor.Moving Average
       Channel:ENiagaraFloatFromLinearColorOptions=::NewEnumerator3
      )
     )
    )
   ) {isSetParameters=true, module=SetVariables_#13}
  ]
  rendererClasses [&NiagaraSpriteRendererProperties]
  renderers [
   {rendererClass=&NiagaraSpriteRendererProperties}
   {rendererClass=&NiagaraSpriteRendererProperties, rendererIndex=1}
   {rendererClass=&NiagaraSpriteRendererProperties, rendererIndex=2}
   {rendererClass=&NiagaraSpriteRendererProperties, rendererIndex=3}
   {rendererClass=&NiagaraSpriteRendererProperties, rendererIndex=4}
  ]
 }
}
summary {
 systemName=NS_MicStand_IndoorLace_Idle_FX
 emitters [
  {
   emitterName=BaseGlows_MicGlow
   rendererClasses [&NiagaraSpriteRendererProperties, &NiagaraMeshRendererProperties]
  }
  {emitterName=NE_SpiralRibbon, rendererClasses [&NiagaraRibbonRendererProperties]}
  {emitterName=Flares_and_Sparkles, rendererClasses [&NiagaraSpriteRendererProperties]}
  {bEnabled=false, emitterName=Stars_and_Motes001, rendererClasses [&NiagaraSpriteRendererProperties]}
 ]
 userVariables [
  User.bIsFrontEndPreview {type=NiagaraBool}
  User.IsInstrument {type=NiagaraFloat, defaultValue=0, description=""}
 ]
}
