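# Niagara swing effect for the first attack of the IndoorLace MicStand pickaxe: a brief warm-gold air streak with sparkles.
# Three mesh emitters burst air cone and airflow shapes from the User skeletal mesh, animated by scale, alpha and material-parameter curves.
# Flares_Sparkles adds two sprite spawn groups with per-group switches, flicker and curl noise; Stars_Motes001 is a disabled variant.
# TOD color tints most layers, and front-end modifiers adjust some alphas.
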NiagaraSystem /SparksCosmetics/Instrument/MicStand/IndoorLace/Effects/NS_Pickaxe_MicStand_IndoorLace_FirstAttack
dependencies {
 usedDataInterfaces [&NiagaraDataInterfaceCurve, &NiagaraDataInterfaceColorCurve]
 usedDynamicInputs []
 usedModules [
  &self:SystemSpawnScript.SystemScriptSource.SystemScriptGraph.NiagaraNodeAssignment_0.SetVariables_#1
  &SystemState
 ]
}
emitters {
 Air_cone001 {
  emitterUpdateScript [
   &EmitterState(Life Cycle Mode=::NewEnumerator1, Loop Behavior=::NewEnumerator1, Loop Delay=0.08, Loop Duration=0.5, UseLoopDelay=true)
   &SpawnBurst_Instantaneous(Use Spawn Probability=false, UseLoopCountLimit=false)
  ]
  particleSpawnScript [
   &InitializeParticle(
    Color=&MultiplyLinearColor(
     Linear Color=<rgba 1 0.807 0.4735 1>
     Scale Factor=&MakeVector4_FromLinearColor(LinearColor=$System.TOD Color)
    )
    Lifetime=0.25
    Mesh Scale=&Multiply_VectorByFloat(Vector=<0.42 0.14 0.14>, Float=1)
    Mesh Scale Mode=::NewEnumerator4
    UsePositionOffset=false
    Write Material Random=true
   )
   &InitialMeshOrientation(Mesh Orientation Mode=::NewEnumerator2, Rotation=<0 0.25 0>, Use Rotation Vector=true)
   &AddVelocity(Velocity=<0 0 -100>) {enabled=false}
   &SkeletalMeshLocation(
    MaskSpawnGroup=false
    Mesh Sampling Type=::NewEnumerator5
    Sampled Position Offset=<0 0 -15>
    Skeletal Mesh=$User.Skeletal Mesh
    Socket Sampling Mode=::NewEnumerator5
   )
  ]
  particleUpdateScript [
   &ParticleState()
   &ScaleMeshSize(Scale Factor=&MakeVector(
     X=&FloatFromCurve(FloatCurve=DataInterface() {propertyValues {Curve {
         defaultValue=1
         keys [
          {arriveTangent=2.6429, interpMode=RCIM_Cubic, leaveTangent=2.6429, tangentMode=RCTM_User, time=0, value=0.8}
          {arriveTangent=-0.5647, interpMode=RCIM_Cubic, leaveTangent=-0.5647, tangentMode=RCTM_User, time=0.2, value=1}
          {arriveTangent=-0.2145, interpMode=RCIM_Cubic, leaveTangent=-0.2145, tangentMode=RCTM_User, time=1, value=0.8}
         ]
        }}})
     Y=&FloatFromCurve(FloatCurve=DataInterface() {propertyValues {Curve {
         defaultValue=1
         keys [
          {arriveTangent=2.4238, interpMode=RCIM_Cubic, leaveTangent=2.4238, tangentMode=RCTM_User, time=0, value=0.8}
          {arriveTangent=-0.1362, interpMode=RCIM_Cubic, leaveTangent=-0.1362, tangentMode=RCTM_User, time=1, value=2}
         ]
        }}})
     Z=&FloatFromCurve(FloatCurve=DataInterface() {propertyValues {Curve {
         defaultValue=1
         keys [
          {arriveTangent=2.4238, interpMode=RCIM_Cubic, leaveTangent=2.4238, tangentMode=RCTM_User, time=0, value=0.8}
          {arriveTangent=-0.1362, interpMode=RCIM_Cubic, leaveTangent=-0.1362, tangentMode=RCTM_User, time=1, value=2}
         ]
        }}})
    ))
   &DynamicMaterialParameters(
    Index 0 Param 1=0.25
    Index 0 Param 2=0.5
    Index 0 Param 3=1
    Index 0 Param 4=&FloatFromCurve(FloatCurve=DataInterface() {propertyValues {Curve {keys [{arriveTangent=-1.3436, interpMode=RCIM_Cubic, leaveTangent=-1.3436, tangentMode=RCTM_User, time=0, value=0.75}, <at 1 0>]}}})
    Index 3 Param 2=1
    Module.Param0WriteEnabled=true
    Module.Param0WriteEnabled3=true
    Module.Param1WriteEnabled=true
    Module.Param1WriteEnabled3=true
    Module.Param2WriteEnabled=true
    Module.Param2WriteEnabled3=true
    Module.Param3WriteEnabled=true
    Module.Param3WriteEnabled3=true
    Write Parameter Index 3=true
   ) {module=DynamicMaterialParameters001}
   &FN_Mesh_Attributes(
    Color=&MakeLinearColorFromVectorAndFloat(
     "Vector (RGB)"=&MakeVectorFromLinearColorRGB(Color=$Particles.Color)
     "Float (Alpha)"=&FloatFromCurve(
      FloatCurve=DataInterface() {propertyValues {Curve {
         defaultValue=1
         keys [
          {arriveTangent=-1.2313, interpMode=RCIM_Cubic, leaveTangent=-1.2313, tangentMode=RCTM_User, time=0, value=1}
          {interpMode=RCIM_Cubic, tangentMode=RCTM_User, time=1, value=0}
         ]
        }}}
      Scale Curve=&FN_FrontEndFloatModifier(Gameplay=1, Front End=1)
     )
    )
    Create Color Attribute=true
    Create Scale Attribute=true
    Scale=&Multiply_Vector(
     A=<0.12 0.25 0.25>
     B=&MakeVector(
      X=&FloatFromCurve(FloatCurve=DataInterface() {propertyValues {Curve {keys [{arriveTangent=0.3763, interpMode=RCIM_Cubic, leaveTangent=0.3763, tangentMode=RCTM_User, time=0, value=1}, <at 1 1.2>]}}})
      Y=&FloatFromCurve(FloatCurve=DataInterface() {propertyValues {Curve {keys [{arriveTangent=3.1084, interpMode=RCIM_Cubic, leaveTangent=3.1084, tangentMode=RCTM_User, time=0, value=0.5}, <at 0.3 1>]}}})
      Z=&FloatFromCurve(FloatCurve=DataInterface() {propertyValues {Curve {keys [{arriveTangent=3.1084, interpMode=RCIM_Cubic, leaveTangent=3.1084, tangentMode=RCTM_User, time=0, value=0.5}, <at 0.3 1>]}}})
     )
    )
    Write to Module Namespace=true
   ) {module=FN_Mesh_Attributes001}
   &ScaleColor(
    Scale Alpha=&FloatFromCurve(
     FloatCurve=DataInterface() {propertyValues {Curve {
        defaultValue=1
        keys [
         {arriveTangent=-1.0675, interpMode=RCIM_Cubic, leaveTangent=-1.0675, tangentMode=RCTM_User, time=0, value=1}
         {interpMode=RCIM_Cubic, tangentMode=RCTM_User, time=1, value=0}
        ]
       }}}
     Scale Curve=&FN_FrontEndFloatModifier(Gameplay=1, Front End=0.2)
    )
    Scale RGB=&MakeVectorFromLinearColorRGB(Color=&ColorFromCurve(ColorCurve=DataInterface() {propertyValues {
        AlphaCurve {keys []}
        BlueCurve {keys [<at 0 1 RCIM_Cubic>,<at 1 0.3958 RCIM_Cubic>]}
        GreenCurve {keys [<at 0 1 RCIM_Cubic>,<at 1 0.3958 RCIM_Cubic>]}
        RedCurve {keys [<at 0 1 RCIM_Cubic>,<at 1 0.3958 RCIM_Cubic>]}
       }}))
    ScaleA=true
    ScaleRGB=true
   )
   /SetVariables(Particles.Dy3:Vector4f=&MakeVector4(X=0, Y=1, Z=0, W=1)) {isSetParameters=true}
   &Drag(Use Linear Drag=true, Use Rotational Drag=true) {enabled=false}
   &SolveForcesAndVelocity(Clamp Velocity=false, Limit Acceleration=false) {enabled=false}
  ]
  rendererClasses [&NiagaraMeshRendererProperties]
  renderers [
   {rendererClass=&NiagaraMeshRendererProperties, rendererIndex=0}
   {rendererClass=&NiagaraMeshRendererProperties, rendererIndex=1}
  ]
 }
 AirflowColor001 {
  emitterUpdateScript [
   &EmitterState(Life Cycle Mode=::NewEnumerator1, Loop Behavior=::NewEnumerator1, Loop Delay=0.07, Loop Duration=0.5, UseLoopDelay=true)
   &SpawnBurst_Instantaneous(Use Spawn Probability=false, UseLoopCountLimit=false)
  ]
  particleSpawnScript [
   &InitializeParticle(
    Color=<rgba 1 0.807 0.4735 1>
    Lifetime=0.2
    Mesh Scale=<0.8 0.8 0.8>
    Mesh Scale Mode=::NewEnumerator4
    UsePositionOffset=false
    Write Material Random=true
   )
   &InitialMeshOrientation(Mesh Orientation Mode=::NewEnumerator2, Rotation=<0 -0.25 0>, Use Rotation Vector=true)
   &SkeletalMeshLocation(
    MaskSpawnGroup=false
    Mesh Sampling Type=::NewEnumerator5
    Orientation Sampling=::NewEnumerator1
    Sampled Position Offset=<0 -1 95>
    Skeletal Mesh=$User.Skeletal Mesh
    Socket Sampling Mode=::NewEnumerator5
   )
   &CameraOffset(Camera Offset Amount=30)
   &AddVelocity(Velocity=<0 0 -100>)
  ]
  particleUpdateScript [
   &ParticleState()
   &ScaleMeshSize(Scale Factor=&MakeVector(
     X=&FloatFromCurve(FloatCurve=DataInterface() {propertyValues {Curve {
         defaultValue=1
         keys [{interpMode=RCIM_Cubic, tangentMode=RCTM_User, time=0, value=0.9}, {interpMode=RCIM_Cubic, tangentMode=RCTM_User, time=1, value=2.5}]
        }}})
     Y=&FloatFromCurve(FloatCurve=DataInterface() {propertyValues {Curve {
         defaultValue=1
         keys [{interpMode=RCIM_Cubic, tangentMode=RCTM_User, time=0, value=1.5}, {interpMode=RCIM_Cubic, tangentMode=RCTM_User, time=1, value=0.6}]
        }}})
     Z=&FloatFromCurve(FloatCurve=DataInterface() {propertyValues {Curve {
         defaultValue=1
         keys [{interpMode=RCIM_Cubic, tangentMode=RCTM_User, time=0, value=1.5}, {interpMode=RCIM_Cubic, tangentMode=RCTM_User, time=1, value=0.6}]
        }}})
    )) {module=ScaleMeshSize001}
   &Drag(Drag=4, Use Linear Drag=true, Use Rotational Drag=true)
   &SolveForcesAndVelocity(Clamp Velocity=false, Limit Acceleration=false)
   &ScaleColor(
    Scale Alpha=&FloatFromCurve(
     FloatCurve=DataInterface() {propertyValues {Curve {keys [<at 0 1 RCIM_Cubic>,<at 1 0>]}}}
     Scale Curve=&FN_FrontEndFloatModifier(Gameplay=0.75, Front End=0.75)
    )
    Scale RGB=&MakeVectorFromLinearColorRGB(Color=$System.TOD Color)
    ScaleA=true
    ScaleRGB=true
   )
   &DynamicMaterialParameters(
    Index 0 Param 1=&FloatFromCurve(FloatCurve=DataInterface() {propertyValues {Curve {keys [<at 0 0.4>,<at 1 0.8>]}}})
    Index 0 Param 2=&FloatFromCurve(FloatCurve=DataInterface() {propertyValues {Curve {keys [
         {arriveTangent=-2.2616, interpMode=RCIM_Cubic, leaveTangent=-3.3361, tangentMode=RCTM_Break, time=0, value=1}
         {arriveTangent=-0.8527, interpMode=RCIM_Cubic, leaveTangent=-0.8527, tangentMode=RCTM_User, time=0.8, value=0}
        ]}}})
    Index 0 Param 3=&FloatFromCurve(FloatCurve=DataInterface() {propertyValues {Curve {keys [<at 0 0.5>,<at 0.65 1>]}}})
    Module.Param0WriteEnabled=true
    Module.Param1WriteEnabled=true
    Module.Param2WriteEnabled=true
    Module.Param3WriteEnabled=true
   ) {module=DynamicMaterialParameters002}
  ]
  rendererClasses [&NiagaraMeshRendererProperties]
  renderers [{rendererClass=&NiagaraMeshRendererProperties, rendererIndex=0}]
 }
 Airflow_Stick001 {
  emitterUpdateScript [
   &EmitterState(Life Cycle Mode=::NewEnumerator1, Loop Behavior=::NewEnumerator1, Loop Delay=0.07, Loop Duration=0.5, UseLoopDelay=true)
   &SpawnBurst_Instantaneous(Use Spawn Probability=false, UseLoopCountLimit=false)
  ]
  particleSpawnScript [
   &InitializeParticle(Color=<rgba 1 0.807 0.4735 1>, Lifetime=0.25, Mesh Scale Mode=::NewEnumerator4, UsePositionOffset=false, Write Material Random=true)
   &AddVelocity(Velocity=<0 0 -100>)
   &InitialMeshOrientation(Mesh Orientation Mode=::NewEnumerator2, Rotation=<0 -0.25 0>, Use Rotation Vector=true)
   &SkeletalMeshLocation(
    Bone Sampling Mode=::NewEnumerator4
    MaskSpawnGroup=false
    Orientation Sampling=::NewEnumerator1
    Sampled Position Offset=<0 0 100>
    Skeletal Mesh=$User.Skeletal Mesh
   )
   &CameraOffset(Camera Offset Amount=20)
  ]
  particleUpdateScript [
   &ParticleState()
   &ScaleMeshSize(Scale Factor=&MakeVector(
     X=&FloatFromCurve(FloatCurve=DataInterface() {propertyValues {Curve {keys [<at 0 1>,<at 1 1.5>]}}})
     Y=1
     Z=1
    ))
   &SolveForcesAndVelocity(Clamp Velocity=false, Limit Acceleration=false)
   &DynamicMaterialParameters(
    Index 0 Param 1=&FloatFromCurve(FloatCurve=DataInterface() {propertyValues {Curve {keys [<at 0 0>,<at 0.5 0>,<at 1 0.08>]}}})
    Index 0 Param 2=&FloatFromCurve(FloatCurve=DataInterface() {propertyValues {Curve {keys [{arriveTangent=-2.2616, interpMode=RCIM_Cubic, time=0, value=-0.5}, {interpMode=RCIM_Cubic, leaveTangent=-0.7156, time=1, value=1}]}}})
    Index 0 Param 3=1.2
    Index 0 Param 4=&FloatFromCurve(FloatCurve=DataInterface() {propertyValues {Curve {keys [<at 0 1>,<at 1 2>]}}}, Scale Curve=0.2)
    Module.Param0WriteEnabled=true
    Module.Param1WriteEnabled=true
    Module.Param2WriteEnabled=true
    Module.Param3WriteEnabled=true
   )
   &ScaleColor(
    Scale Alpha=&FloatFromCurve(FloatCurve=DataInterface() {propertyValues {Curve {keys [<at 0 1 RCIM_Cubic>, {tangentMode=RCTM_User, time=1, value=0}]}}}, Scale Curve=0.75)
    ScaleA=true
    ScaleRGB=false
   )
  ]
  rendererClasses [&NiagaraMeshRendererProperties]
  renderers [{rendererClass=&NiagaraMeshRendererProperties, rendererIndex=0}]
 }
 Flares_Sparkles {
  emitterUpdateScript [
   &EmitterState(Life Cycle Mode=::NewEnumerator1, Loop Behavior=::NewEnumerator1, Loop Delay=0.06, Loop Duration=0.5, UseLoopDelay=true)
   &SpawnBurst_Instantaneous(Spawn Count=&RandomRangeInt(Minimum=6, Maximum=9), Use Spawn Probability=false, UseLoopCountLimit=true)
   &SpawnBurst_Instantaneous(
    Spawn Count=&RandomRangeInt(Minimum=6, Maximum=8)
    Spawn Group=1
    Use Spawn Probability=false
    UseLoopCountLimit=true
   ) {module=SpawnBurst_Instantaneous001}
  ]
  particleSpawnScript [
   /SetVariables(Particles.VisibilityTag=&SpawnGroup()) {isSetParameters=true, module=SetVariables_#2}
   /SetVariables(Particles.Variance:NiagaraFloat=&MergeHelper_FloatSwitch(
     0=&FloatFromCurve(
      FloatCurve=DataInterface() {propertyValues {Curve {keys [<at 0 1>,<at 1 0>]}}}
      CurveIndex=&ReturnNormalizedExecIndex()
     )
     1=&FloatFromCurve(
      FloatCurve=DataInterface() {propertyValues {Curve {keys [<at 0 1>,<at 1 0>]}}}
      CurveIndex=&ReturnNormalizedExecIndex()
     )
     2=0
     3=?
     Index=&SpawnGroup()
     Num=::NewEnumerator1
    )) {isSetParameters=true, module=SetVariables_#3}
   &InitializeParticle(
    Color=&MergeHelper_ColorSwitch(
     0=&RandomRangeLinearColor(Minimum=<rgba 1 0.7676 0.4189 0.5>, Maximum=<rgba 1.185 1.25 0.5892 1>)
     1=&RandomRangeLinearColor(Minimum=<rgba 4 3.0703 1.6758 1>, Maximum=<rgba 4.7402 5 2.3568 1>)
     2=?
     3=?
     Index=&SpawnGroup()
     Num=::NewEnumerator0
    )
    Lifetime=&MergeHelper_FloatSwitch(
     0=&RandomRangeFloat(Minimum=0.25, Maximum=0.45)
     1=&FloatFromCurve(FloatCurve=DataInterface() {propertyValues {Curve {keys [<at 0 0.3>,<at 1 0.2>]}}}, CurveIndex=$Particles.Variance)
     2=&RandomRangeFloat(Minimum=0.5, Maximum=0.7)
     3=?
     Index=&SpawnGroup()
     Num=::NewEnumerator1
    )
    Sprite Size Mode=::NewEnumerator3
    Uniform Sprite Size=&MergeHelper_FloatSwitch(
     0=&RandomRangeFloat(Minimum=0.5, Maximum=1)
     1=&FloatFromCurve(FloatCurve=DataInterface() {propertyValues {Curve {keys [<at 0 8>,<at 1 12>]}}}, CurveIndex=$Particles.Variance)
     2=&Multiply_Float(A=&RandomRangeFloat(Minimum=1.5, Maximum=3), B=1)
     3=?
     Index=&SpawnGroup()
     Num=::NewEnumerator0
    )
    UsePositionOffset=false
    Write Material Random=true
   )
   &ShapeLocation(
    Box Size=&MergeHelper_VectorSwitch(0=<10 10 70>, 1=<25 25 60>, 2=?, 3=?, Index=&SpawnGroup(), Num=::NewEnumerator0)
    MaskSpawnGroup=false
    Override Seed=false
    Shape Origin=&VectorFromFloat(Value=0)
    Shape Primitive=::NewEnumerator2
    Surface Only=false
    UseOwnerScale=true
   )
   &SkeletalMeshLocation(
    Bone Sampling Mode=::NewEnumerator4
    MaskSpawnGroup=false
    Sampled Position Offset=&AddVectorToPosition(
     Position=$Particles.ShapeLocation.ShapePosition
     Vector=&MergeHelper_VectorSwitch(0=<0 0 -5>, 1=<0 0 -10>, 2=<0 0 0>, 3=?, Index=&SpawnGroup(), Num=::NewEnumerator1)
    )
    Skeletal Mesh=$User.Skeletal Mesh
   )
   &AddVelocity(Module.Use Falloff Distance=false, Override Seed=false, Velocity Mode=::NewEnumerator1) {enabled=false}
   &AddVelocity(
    Velocity=<0 0 -500>
    Velocity Speed Scale=&MergeHelper_FloatSwitch(
     0=&RandomRangeFloat(Minimum=0.75, Maximum=1.5)
     1=&RandomRangeFloat(Minimum=0.8, Maximum=1.25)
     2=&RandomRangeFloat(Minimum=1, Maximum=1.25)
     3=?
     Index=&SpawnGroup()
     Num=::NewEnumerator1
    )
   ) {module=AddVelocity001}
   &CameraOffset(Camera Offset Amount=20)
  ]
  particleUpdateScript [
   &ParticleState()
   /SetVariables(Particles.FlickerWaveform:NiagaraFloat=&Waveform(
     Global Phase=&Add_Float(A=$Engine.Time, B=&RandomRangeFloat(Minimum=0, Maximum=1))
     "Waveform [1]"=::NewEnumerator10
     "[1] Frequency"=10
     "[1] Amplitude Min/Max"=<0.5 1.4>
    )) {isSetParameters=true, module=SetVariables_#4}
   &ScaleSpriteSize(
    Scale Sprite Size Mode=::NewEnumerator0
    Uniform Scale Factor=&MergeHelper_FloatSwitch(
     0=&FloatFromCurve(
      FloatCurve=DataInterface() {propertyValues {Curve {
         defaultValue=1
         keys [
          {arriveTangent=9.0725, interpMode=RCIM_Cubic, leaveTangent=9.0725, tangentMode=RCTM_User, time=0, value=0}
          {arriveTangent=0.1349, interpMode=RCIM_Cubic, leaveTangent=0.1349, tangentMode=RCTM_User, time=0.1408, value=0.9847}
          {arriveTangent=-0.0405, interpMode=RCIM_Cubic, leaveTangent=-0.0405, tangentMode=RCTM_User, time=1, value=0}
         ]
        }}}
      Scale Curve=$Particles.FlickerWaveform
     )
     1=&FloatFromCurve(FloatCurve=DataInterface() {propertyValues {Curve {
         defaultValue=0.2
         keys [
          {arriveTangent=9.0725, interpMode=RCIM_Cubic, leaveTangent=9.0725, tangentMode=RCTM_User, time=0, value=0}
          {arriveTangentWeight=0.0469, interpMode=RCIM_Cubic, leaveTangentWeight=0.2864, tangentMode=RCTM_User, time=0.1408, value=1}
          {arriveTangent=-2.7648, interpMode=RCIM_Cubic, leaveTangent=-2.7648, tangentMode=RCTM_User, time=1, value=0}
         ]
        }}})
     2=&FloatFromCurve(
      FloatCurve=DataInterface() {propertyValues {Curve {
         defaultValue=0
         keys [
          {interpMode=RCIM_Cubic, tangentMode=RCTM_User, time=0, value=0}
          {interpMode=RCIM_Cubic, tangentMode=RCTM_User, time=0.3, value=1}
          {tangentMode=RCTM_User, time=1, value=0}
         ]
        }}}
      Scale Curve=&RandomRangeFloat(Minimum=0.75, Maximum=1)
     )
     3=?
     Index=&SpawnGroup()
     Num=::NewEnumerator0
    )
   )
   &CurlNoiseForce(
    Bias Noise=true
    Bias Noise Field:Vector3f=&MergeHelper_VectorSwitch(0=<0 0 -0.5>, 1=<0 0 0>, 2=<0 0 0>, 3=?, Index=&SpawnGroup(), Num=::NewEnumerator1)
    Noise Frequency=&MergeHelper_FloatSwitch(
     0=12
     1=0
     2=&RandomRangeFloat(Minimum=20, Maximum=30)
     3=?
     Index=&SpawnGroup()
     Num=::NewEnumerator1
    )
    Noise Strength=&MergeHelper_FloatSwitch(
     0=&RandomRangeFloat(Minimum=60, Maximum=175)
     1=0
     2=&RandomRangeFloat(Minimum=75, Maximum=300)
     3=?
     Index=&SpawnGroup()
     Num=::NewEnumerator1
    )
    Pan Noise=false
   )
   &Drag(
    Drag=&FloatFromCurve(FloatCurve=DataInterface() {propertyValues {Curve {keys [<at 0 2.5>,<at 1 8>]}}})
    Use Linear Drag=true
    Use Rotational Drag=true
   )
   &SolveForcesAndVelocity(Clamp Velocity=false, Limit Acceleration=false)
   &ScaleColor(
    Scale Alpha=&MergeHelper_FloatSwitch(
     0=&Multiply_Float(
      A=&FloatFromCurve(FloatCurve=DataInterface() {propertyValues {Curve {
          defaultValue=1
          keys [
           {arriveTangent=9.0725, interpMode=RCIM_Cubic, leaveTangent=9.0725, tangentMode=RCTM_User, time=0, value=0}
           {arriveTangent=-2.5249, interpMode=RCIM_Cubic, leaveTangent=-2.5249, tangentMode=RCTM_User, time=0.1408, value=0.9847}
           {arriveTangent=-0.0405, interpMode=RCIM_Cubic, leaveTangent=-0.0405, tangentMode=RCTM_User, time=1, value=0}
          ]
         }}})
      B=&Waveform(
       "Waveform [1]"=::NewEnumerator0
       "[1] Frequency"=&FloatFromCurve(
        FloatCurve=DataInterface() {propertyValues {Curve {keys [<at 0.2 0>,<at 0.4 1>]}}}
        Scale Curve=&RandomRangeFloat(Minimum=3, Maximum=5)
       )
       "[1] Amplitude Min/Max"=<0 1>
      )
     )
     1=&FloatFromCurve(FloatCurve=DataInterface() {propertyValues {Curve {keys [<at 0.3 1 RCIM_Cubic>, {arriveTangentWeight=0.4442, interpMode=RCIM_Cubic, tangentMode=RCTM_User, time=1, value=0}]}}})
     2=&Multiply_Float(
      A=&FloatFromCurve(
       FloatCurve=DataInterface() {propertyValues {Curve {keys [<at 0 1 RCIM_Cubic>, {arriveTangent=-2.4387, interpMode=RCIM_Cubic, leaveTangent=-2.4387, tangentMode=RCTM_User, time=1, value=0}]}}}
       Scale Curve=&Waveform("Waveform [1]"=::NewEnumerator0, "[1] Frequency"=4, "[1] Amplitude Min/Max"=<0.1 1.1>)
      )
      B=&FN_FrontEndFloatModifier(Gameplay=1, Front End=0.1)
     )
     3=?
     Index=&SpawnGroup()
     Num=::NewEnumerator0
    )
    Scale RGB=&MergeHelper_VectorSwitch(
     0=&Multiply_Vector(
      A=&MakeVectorFromLinearColorRGB(Color=$System.TOD Color)
      B=&VectorFromFloat(Value=&FloatFromCurve(FloatCurve=DataInterface() {propertyValues {Curve {
           defaultValue=1
           keys [
            {arriveTangent=47.9314, interpMode=RCIM_Cubic, leaveTangent=47.9314, tangentMode=RCTM_User, time=0, value=1}
            {arriveTangent=-4.3766, interpMode=RCIM_Cubic, leaveTangent=-4.3766, tangentMode=RCTM_User, time=0.1, value=2.334}
            {
             arriveTangent=-0.6181
             arriveTangentWeight=0.377
             interpMode=RCIM_Cubic
             leaveTangent=-0.6181
             leaveTangentWeight=0.2187
             tangentMode=RCTM_User
             time=1
             value=0.95
            }
           ]
          }}}))
     )
     1=&Multiply_Vector(
      A=&MakeVectorFromLinearColorRGB(Color=$System.TOD Color)
      B=&VectorFromFloat(Value=&FloatFromCurve(FloatCurve=DataInterface() {propertyValues {Curve {
           defaultValue=1
           keys [
            {arriveTangent=13.3606, interpMode=RCIM_Cubic, leaveTangent=13.3606, tangentMode=RCTM_User, time=0, value=1}
            {arriveTangent=-1.753, interpMode=RCIM_Cubic, leaveTangent=-1.753, tangentMode=RCTM_User, time=0.1, value=1.5}
            {arriveTangentWeight=0.317, interpMode=RCIM_Cubic, leaveTangentWeight=0.2187, tangentMode=RCTM_User, time=1, value=0.95}
           ]
          }}}))
     )
     2=?
     3=?
     Index=&SpawnGroup()
     Num=::NewEnumerator0
    )
    ScaleA=true
    ScaleRGB=true
   )
   /SetVariables(
    Particles.Glow Color=&MergeHelper_ColorSwitch(
     0=&MakeLinearColorFromVectorAndFloat(
      "Vector (RGB)"=&MakeVectorFromLinearColorRGB(Color=&MultiplyLinearColor(Linear Color=$Particles.Color, Scale Factor=&MakeVector4_FromLinearColor(LinearColor=$Particles.Color)))
      "Float (Alpha)"=0.08
     )
     1=&MakeLinearColorFromVectorAndFloat(
      "Vector (RGB)"=&MakeVectorFromLinearColorRGB(Color=&MultiplyLinearColor(
        Linear Color=$Particles.Color
        Scale Factor=&MakeVector4_FromLinearColor(LinearColor=<rgba 0.15 0.0919 0.0339 1>)
       ))
      "Float (Alpha)"=0.5
     )
     2=?
     3=?
     Index=&SpawnGroup()
     Num=::NewEnumerator0
    )
    Particles.Glow Size=&Multiply_Vector2D(
     A=$Particles.SpriteSize
     B=&MergeHelper_Vector2Switch(0=<3.5 4.5>, 1=<0.5 0.5>, 2=?, 3=?, 4=?, Index=&SpawnGroup(), Num=::NewEnumerator0)
    )
   ) {isSetParameters=true, module=SetVariables_#5}
  ]
  rendererClasses [&NiagaraSpriteRendererProperties]
  renderers [
   {rendererClass=&NiagaraSpriteRendererProperties}
   {rendererClass=&NiagaraSpriteRendererProperties, rendererIndex=1}
   {rendererClass=&NiagaraSpriteRendererProperties, rendererIndex=2}
   {rendererClass=&NiagaraSpriteRendererProperties, rendererIndex=3}
  ]
 }
 Stars_Motes001 {
  bEnabled=false
  emitterUpdateScript [
   &EmitterState(Life Cycle Mode=::NewEnumerator1, Loop Behavior=::NewEnumerator1, Loop Delay=0.06, Loop Duration=0.5, UseLoopDelay=true)
   &SpawnBurst_Instantaneous(Spawn Count=&RandomRangeInt(Minimum=6, Maximum=9), Use Spawn Probability=false, UseLoopCountLimit=true)
   &SpawnBurst_Instantaneous(
    Spawn Count=&RandomRangeInt(Minimum=3, Maximum=4)
    Spawn Group=1
    Use Spawn Probability=false
    UseLoopCountLimit=true
   ) {module=SpawnBurst_Instantaneous001}
  ]
  particleSpawnScript [
   /SetVariables(Particles.VisibilityTag=&SpawnGroup()) {isSetParameters=true, module=SetVariables_#2}
   /SetVariables(Particles.Variance:NiagaraFloat=&MergeHelper_FloatSwitch(
     0=&FloatFromCurve(
      FloatCurve=DataInterface() {propertyValues {Curve {keys [<at 0 1>,<at 1 0>]}}}
      CurveIndex=&ReturnNormalizedExecIndex()
     )
     1=&FloatFromCurve(
      FloatCurve=DataInterface() {propertyValues {Curve {keys [<at 0 1>,<at 1 0>]}}}
      CurveIndex=&ReturnNormalizedExecIndex()
     )
     2=0
     3=?
     Index=&SpawnGroup()
     Num=::NewEnumerator1
    )) {isSetParameters=true, module=SetVariables_#3}
   &InitializeParticle(
    Color=&MergeHelper_ColorSwitch(
     0=&RandomRangeLinearColor(Minimum=<rgba 1 0.7676 0.4189 0.5>, Maximum=<rgba 1.185 1.25 0.5892 1>)
     1=&RandomRangeLinearColor(Minimum=<rgba 1 0.8125 0.5312 0.7>, Maximum=<rgba 0.959 1 0.5833 1>)
     2=?
     3=?
     Index=&SpawnGroup()
     Num=::NewEnumerator0
    )
    Lifetime=&MergeHelper_FloatSwitch(
     0=&RandomRangeFloat(Minimum=0.25, Maximum=0.45)
     1=&FloatFromCurve(FloatCurve=DataInterface() {propertyValues {Curve {keys [<at 0 0.3>,<at 1 0.2>]}}}, CurveIndex=$Particles.Variance)
     2=&RandomRangeFloat(Minimum=0.5, Maximum=0.7)
     3=?
     Index=&SpawnGroup()
     Num=::NewEnumerator1
    )
    Sprite Rotation Angle=&MergeHelper_FloatSwitch(
     0=45
     1=&RandomRangeFloat(Minimum=0, Maximum=360)
     2=?
     3=?
     Index=&SpawnGroup()
     Num=::NewEnumerator0
    )
    Sprite Rotation Mode=::NewEnumerator0
    Sprite Size Mode=::NewEnumerator3
    Uniform Sprite Size=&MergeHelper_FloatSwitch(
     0=&RandomRangeFloat(Minimum=0.2, Maximum=1)
     1=&FloatFromCurve(FloatCurve=DataInterface() {propertyValues {Curve {keys [<at 0 4>,<at 1 7>]}}}, CurveIndex=$Particles.Variance)
     2=&Multiply_Float(A=&RandomRangeFloat(Minimum=1.5, Maximum=3), B=1)
     3=?
     Index=&SpawnGroup()
     Num=::NewEnumerator0
    )
    UsePositionOffset=false
    Write Material Random=true
   )
   &ShapeLocation(
    Box Size=&MergeHelper_VectorSwitch(0=<10 10 70>, 1=<25 25 40>, 2=<15 15 50>, 3=?, Index=&SpawnGroup(), Num=::NewEnumerator1)
    MaskSpawnGroup=false
    Override Seed=false
    Shape Origin=&VectorFromFloat(Value=0)
    Shape Primitive=::NewEnumerator2
    Surface Only=false
    UseOwnerScale=true
   )
   &SkeletalMeshLocation(
    Bone Sampling Mode=::NewEnumerator4
    MaskSpawnGroup=false
    Sampled Position Offset=&AddVectorToPosition(
     Position=$Particles.ShapeLocation.ShapePosition
     Vector=&MergeHelper_VectorSwitch(0=<0 0 -5>, 1=<0 0 -10>, 2=<0 0 0>, 3=?, Index=&SpawnGroup(), Num=::NewEnumerator1)
    )
    Skeletal Mesh=$User.Skeletal Mesh
   )
   &AddVelocity(Module.Use Falloff Distance=false, Override Seed=false, Velocity Mode=::NewEnumerator1) {enabled=false}
   &AddVelocity(
    Velocity=<0 0 -500>
    Velocity Speed Scale=&MergeHelper_FloatSwitch(
     0=&RandomRangeFloat(Minimum=0.75, Maximum=1.5)
     1=&RandomRangeFloat(Minimum=0.8, Maximum=1.25)
     2=&RandomRangeFloat(Minimum=1, Maximum=1.25)
     3=?
     Index=&SpawnGroup()
     Num=::NewEnumerator1
    )
   ) {module=AddVelocity001}
   &CameraOffset(Camera Offset Amount=20)
  ]
  particleUpdateScript [
   &ParticleState()
   /SetVariables(Particles.FlickerWaveform:NiagaraFloat=&Waveform(
     Global Phase=&Add_Float(A=$Engine.Time, B=&RandomRangeFloat(Minimum=0, Maximum=1))
     "Waveform [1]"=::NewEnumerator10
     "[1] Frequency"=10
     "[1] Amplitude Min/Max"=<0.5 1.4>
    )) {isSetParameters=true, module=SetVariables_#4}
   &ScaleSpriteSize(
    Scale Sprite Size Mode=::NewEnumerator0
    Uniform Scale Factor=&MergeHelper_FloatSwitch(
     0=&FloatFromCurve(
      FloatCurve=DataInterface() {propertyValues {Curve {
         defaultValue=1
         keys [
          {arriveTangent=9.0725, interpMode=RCIM_Cubic, leaveTangent=9.0725, tangentMode=RCTM_User, time=0, value=0}
          {arriveTangent=-2.5249, interpMode=RCIM_Cubic, leaveTangent=-2.5249, tangentMode=RCTM_User, time=0.1408, value=0.9847}
          {arriveTangent=-0.0405, interpMode=RCIM_Cubic, leaveTangent=-0.0405, tangentMode=RCTM_User, time=1, value=0}
         ]
        }}}
      Scale Curve=$Particles.FlickerWaveform
     )
     1=&FloatFromCurve(FloatCurve=DataInterface() {propertyValues {Curve {
         defaultValue=0.2
         keys [
          {arriveTangent=9.0725, interpMode=RCIM_Cubic, leaveTangent=9.0725, tangentMode=RCTM_User, time=0, value=0}
          {arriveTangentWeight=0.1274, interpMode=RCIM_Cubic, leaveTangentWeight=0.7778, tangentMode=RCTM_User, time=0.1408, value=0.9847}
          {arriveTangent=-4.3294, interpMode=RCIM_Cubic, leaveTangent=-4.3294, tangentMode=RCTM_User, time=1, value=0}
         ]
        }}})
     2=&FloatFromCurve(
      FloatCurve=DataInterface() {propertyValues {Curve {
         defaultValue=0
         keys [
          {interpMode=RCIM_Cubic, tangentMode=RCTM_User, time=0, value=0}
          {interpMode=RCIM_Cubic, tangentMode=RCTM_User, time=0.3, value=1}
          {tangentMode=RCTM_User, time=1, value=0}
         ]
        }}}
      Scale Curve=&RandomRangeFloat(Minimum=0.75, Maximum=1)
     )
     3=?
     Index=&SpawnGroup()
     Num=::NewEnumerator0
    )
   )
   &CurlNoiseForce(
    Bias Noise=true
    Bias Noise Field:Vector3f=&MergeHelper_VectorSwitch(0=<0 0 -0.5>, 1=<0 0 0>, 2=<0 0 0>, 3=?, Index=&SpawnGroup(), Num=::NewEnumerator1)
    Noise Frequency=&MergeHelper_FloatSwitch(
     0=12
     1=0
     2=&RandomRangeFloat(Minimum=20, Maximum=30)
     3=?
     Index=&SpawnGroup()
     Num=::NewEnumerator1
    )
    Noise Strength=&MergeHelper_FloatSwitch(
     0=&RandomRangeFloat(Minimum=60, Maximum=175)
     1=0
     2=&RandomRangeFloat(Minimum=75, Maximum=300)
     3=?
     Index=&SpawnGroup()
     Num=::NewEnumerator1
    )
    Pan Noise=false
   )
   &GravityForce(Gravity=&MergeHelper_VectorSwitch(
     0=<0 0 0>
     1=<0 0 0>
     2=&Multiply_VectorByFloat(
      Vector=<0 0 -300>
      Float=&FloatFromCurve(FloatCurve=DataInterface() {propertyValues {Curve {
          defaultValue=1
          keys [
           <at 0.1191 0 RCIM_Cubic>
           {arriveTangent=-0.7473, interpMode=RCIM_Cubic, leaveTangent=-0.7473, tangentMode=RCTM_User, time=0.3761, value=0.9394}
          ]
         }}})
     )
     3=?
     Index=&SpawnGroup()
     Num=::NewEnumerator1
    ))
   &Drag(
    Drag=&FloatFromCurve(FloatCurve=DataInterface() {propertyValues {Curve {keys [<at 0 2.5>,<at 1 8>]}}})
    Use Linear Drag=true
    Use Rotational Drag=true
   )
   &SolveForcesAndVelocity(Clamp Velocity=false, Limit Acceleration=false)
   &ScaleColor(
    Scale Alpha=&MergeHelper_FloatSwitch(
     0=&Multiply_Float(
      A=&FloatFromCurve(FloatCurve=DataInterface() {propertyValues {Curve {
          defaultValue=1
          keys [
           {arriveTangent=9.0725, interpMode=RCIM_Cubic, leaveTangent=9.0725, tangentMode=RCTM_User, time=0, value=0}
           {arriveTangent=-2.5249, interpMode=RCIM_Cubic, leaveTangent=-2.5249, tangentMode=RCTM_User, time=0.1408, value=0.9847}
           {arriveTangent=-0.0405, interpMode=RCIM_Cubic, leaveTangent=-0.0405, tangentMode=RCTM_User, time=1, value=0}
          ]
         }}})
      B=&Waveform(
       "Waveform [1]"=::NewEnumerator0
       "[1] Frequency"=&FloatFromCurve(
        FloatCurve=DataInterface() {propertyValues {Curve {keys [<at 0.2 0>,<at 0.4 1>]}}}
        Scale Curve=&RandomRangeFloat(Minimum=3, Maximum=5)
       )
       "[1] Amplitude Min/Max"=<0 1>
      )
     )
     1=&FloatFromCurve(FloatCurve=DataInterface() {propertyValues {Curve {keys [<at 0.3 1 RCIM_Cubic>, {arriveTangentWeight=0.4442, interpMode=RCIM_Cubic, tangentMode=RCTM_User, time=1, value=0}]}}})
     2=&Multiply_Float(
      A=&FloatFromCurve(
       FloatCurve=DataInterface() {propertyValues {Curve {keys [<at 0 1 RCIM_Cubic>, {arriveTangent=-2.4387, interpMode=RCIM_Cubic, leaveTangent=-2.4387, tangentMode=RCTM_User, time=1, value=0}]}}}
       Scale Curve=&Waveform("Waveform [1]"=::NewEnumerator0, "[1] Frequency"=4, "[1] Amplitude Min/Max"=<0.1 1.1>)
      )
      B=&FN_FrontEndFloatModifier(Gameplay=1, Front End=0.1)
     )
     3=?
     Index=&SpawnGroup()
     Num=::NewEnumerator0
    )
    Scale RGB=&MergeHelper_VectorSwitch(
     0=&Multiply_Vector(
      A=&MakeVectorFromLinearColorRGB(Color=$System.TOD Color)
      B=&VectorFromFloat(Value=&FloatFromCurve(FloatCurve=DataInterface() {propertyValues {Curve {
           defaultValue=1
           keys [
            {arriveTangent=47.9314, interpMode=RCIM_Cubic, leaveTangent=47.9314, tangentMode=RCTM_User, time=0, value=1}
            {arriveTangent=-4.3766, interpMode=RCIM_Cubic, leaveTangent=-4.3766, tangentMode=RCTM_User, time=0.1, value=2.334}
            {
             arriveTangent=-0.6181
             arriveTangentWeight=0.377
             interpMode=RCIM_Cubic
             leaveTangent=-0.6181
             leaveTangentWeight=0.2187
             tangentMode=RCTM_User
             time=1
             value=0.95
            }
           ]
          }}}))
     )
     1=&Multiply_Vector(
      A=&MakeVectorFromLinearColorRGB(Color=$System.TOD Color)
      B=&VectorFromFloat(Value=&FloatFromCurve(FloatCurve=DataInterface() {propertyValues {Curve {
           defaultValue=1
           keys [
            {arriveTangent=13.3606, interpMode=RCIM_Cubic, leaveTangent=13.3606, tangentMode=RCTM_User, time=0, value=1}
            {arriveTangent=-1.753, interpMode=RCIM_Cubic, leaveTangent=-1.753, tangentMode=RCTM_User, time=0.1, value=1.5}
            {arriveTangentWeight=0.317, interpMode=RCIM_Cubic, leaveTangentWeight=0.2187, tangentMode=RCTM_User, time=1, value=0.95}
           ]
          }}}))
     )
     2=?
     3=?
     Index=&SpawnGroup()
     Num=::NewEnumerator0
    )
    ScaleA=true
    ScaleRGB=true
   )
   /SetVariables(
    Particles.Glow Color=&MergeHelper_ColorSwitch(
     0=&MakeLinearColorFromVectorAndFloat(
      "Vector (RGB)"=&MakeVectorFromLinearColorRGB(Color=&MultiplyLinearColor(Linear Color=$Particles.Color, Scale Factor=&MakeVector4_FromLinearColor(LinearColor=$Particles.Color)))
      "Float (Alpha)"=0.08
     )
     1=&MakeLinearColorFromVectorAndFloat(
      "Vector (RGB)"=&MakeVectorFromLinearColorRGB(Color=&MultiplyLinearColor(Linear Color=$Particles.Color, Scale Factor=&MakeVector4_FromLinearColor(LinearColor=$Particles.Color)))
      "Float (Alpha)"=0.3
     )
     2=?
     3=?
     Index=&SpawnGroup()
     Num=::NewEnumerator0
    )
    Particles.Glow Size=&Multiply_Vector2D(
     A=$Particles.SpriteSize
     B=&MergeHelper_Vector2Switch(0=<3.5 4.5>, 1=<1.75 1.75>, 2=?, 3=?, 4=?, Index=&SpawnGroup(), Num=::NewEnumerator0)
    )
   ) {isSetParameters=true, module=SetVariables_#5}
  ]
  rendererClasses [&NiagaraSpriteRendererProperties]
  renderers [
   {rendererClass=&NiagaraSpriteRendererProperties}
   {rendererClass=&NiagaraSpriteRendererProperties, rendererIndex=1}
   {rendererClass=&NiagaraSpriteRendererProperties, rendererIndex=2}
   {rendererClass=&NiagaraSpriteRendererProperties, rendererIndex=3}
  ]
 }
}
summary {
 systemName=NS_Pickaxe_MicStand_IndoorLace_FirstAttack
 emitters [
  {emitterName=Airflow_Stick001, rendererClasses [&NiagaraMeshRendererProperties]}
  {emitterName=AirflowColor001, rendererClasses [&NiagaraMeshRendererProperties]}
  {emitterName=Air_cone001, rendererClasses [&NiagaraMeshRendererProperties]}
  {emitterName=Flares_Sparkles, rendererClasses [&NiagaraSpriteRendererProperties]}
  {bEnabled=false, emitterName=Stars_Motes001, rendererClasses [&NiagaraSpriteRendererProperties]}
 ]
 userVariables [
  User.bIsFrontEnd {type=NiagaraBool}
  User.bIsFrontEndPreview {type=NiagaraBool}
  User.HitColor {type=LinearColor, description="", defaultValue {struct=LinearColor, value=<rgba 0.2634 0.9002 1 1>}}
  User.IsVariant {type=NiagaraFloat, defaultValue=0, description=""}
  User.Skeletal Mesh {
   type=NiagaraDataInterfaceSkeletalMesh
   defaultValue {
    struct=NiagaraExt_VariableValue_DataInterface
    value {dataInterface=&self:NiagaraDataInterfaceSkeletalMesh_0}
   }
  }
 ]
}
